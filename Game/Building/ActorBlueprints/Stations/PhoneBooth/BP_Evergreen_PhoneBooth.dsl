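; Phone booth station that a player enters to change character skin, with swinging door and animated blinds.
; On authority, entering allows a loadout refresh, then either applies the locker effect (debug MCP-swap flag) or records the original character and applies skin-override and cooldown GameplayEffects.
; Enter and exit each use one combined multicast to drive door and blind timelines.
; Interaction requires box overlap, no blocked tags and an enabled booth.
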(event Collision|EventActorBeginOverlap
  (params (OtherActor :type Actor%20Object%20Reference))
  (|CallParentFunction OtherActor) ; This node is disabled and will not be called.\nDrag off pins to build functionality.
)

(event EventTick
  (params (DeltaSeconds :type Float%20%28single-precision%29))
  (|CallParentFunction DeltaSeconds) ; This node is disabled and will not be called.\nDrag off pins to build functionality.
)

(event EventBeginPlay
  (bind _cylinderdistance (Variables|Default|GetCylinderDistance))
  ;[0 1 0] BeginPlay
  (Class|FortAsyncActionCurrentPlaylistReady|PlaylistDataReadyAsync
    (:OnReady
      (CallFunction|InitializeUI)
      (Variables|Default|SetBoothOcupied (Variables|Default|GetBoothOcupied)) ; Resets Blinds Material
      (Rendering|SetVisibility _cylinderdistance false false)
      (Development|SetHiddeninGame _cylinderdistance false false))))

(event EventInteractEnter
  (bind _hidingplayers (Variables|Default|GetHidingPlayers))
  (bind _returnvalue (Utilities|Array|LastIndex _hidingplayers))
  (bind _output (Utilities|Array|Get(acopy) _hidingplayers _returnvalue))
  ;[1 1 1] Event -> Interact Enter - Call the events that open the Locker Widget when player enters the booth
  (Networking|SwitchHasAuthority
    (:Authority
      (Development|PrintString "Open the Locker")
      (Utilities|IsValid _output
        (:Is%20Valid
          (CallFunction|PrepareforLoadoutRefresh _output))))))

(event EventPrepareForLoadoutRefresh
  (params (InteractingPawn :type Fort%20Pawn%20Object%20Reference))
  (bind _hidingplayers (Variables|Default|GetHidingPlayers))
  (bind _returnvalue (Utilities|Array|LastIndex _hidingplayers))
  (bind _returnvalue_2 (Ability|GetFortAbilitySystemComponent InteractingPawn))
  (bind _self self)
  (bind _returnvalue_3 (Ability|GetFortAbilitySystemComponent _self))
  (bind _output (Utilities|Array|Get(acopy) _hidingplayers _returnvalue))
  (bind _self_2 self)
  ;[0.03646 0.03646 0.03646] Event PrepareforLoadoutRefresh
  (Utilities|IsValid InteractingPawn
    (:Is%20Valid
      (bind _asfort_player_controller_zone (Utilities|Casting|CastToFortPlayerControllerZone (Fortnite|GetFortPlayerControllerfromActor InteractingPawn)))
      (Cosmetics|AllowLoadoutRefresh _asfort_player_controller_zone)
      (CallFunction|BoothAnimations true)
      (if (Variables|Default|GetbDebugUseMcpSwap)
        ;;[1 1 1] Apply GE that gives GA that handles the whole phone booth sequence MCP
        (Utilities|IsValid _returnvalue_2
          (:Is%20Valid
            (GameplayEffects|FortApplyGameplayEffectToTarget (Ability|GetFortAbilitySystemComponent self) (Variables|Default|GetGE_EnterLocker) _returnvalue_2 0 self self 0.0 (GameplayEffects|MakeFortEffectContext _returnvalue_2 0 self self 0))))
        (else
          ;[1 1 1] Apply GE that gives GA that handles the whole phone booth sequence
          (CallFunction|AddPawnandLoadouttoArrays InteractingPawn)
          (Utilities|IsValid _output
            (:Is%20Valid
              (bind _output_get (Variables|Default|SetExitingPlayer _output))
              (if (Networking|IsServer)
                (Utilities|IsValid (Ability|GetFortAbilitySystemComponent _output_get)
                  (:Is%20Valid
                    (GameplayEffects|FortApplyGameplayEffectToTarget _returnvalue_3 (Variables|Default|GetGE_StaticSkinOverride) (Ability|GetFortAbilitySystemComponent _output_get) 0 _self_2 _self_2 0.0 (GameplayEffects|MakeFortEffectContext (Ability|GetFortAbilitySystemComponent _output_get) 0 _self_2 _self_2 0))
                    (GameplayEffects|ApplyGameplayEffectToTarget _returnvalue_3 (Variables|Default|GetGE_PhoneBooth_Cooldown) (Ability|GetFortAbilitySystemComponent _output_get) 0.0 (Class|AthenaFunctionLibrary|MakeSimpleFortEffectContext _self_2 _self_2))))))))))))

(event Custom|BoothAnimations
  (params (bEntering :type Boolean))
  ;[1 1 1] Send RPCs for the door and blinds anim . Seems wasteful to call 2 multi casts. There should just be BoothEnterAnimations Multi , and BoothExitAnimations Multi
  (if bEntering
    (CallFunction|EnterBoothAnimations)
    (else
      (CallFunction|ExitBoothAnimations)))
  (Networking|FlushNetDormancy))

(event Custom|PhoneBoothBlindsEnter
  (bind _staticmeshcomponent (Variables|Debug|GetStaticMeshComponent))
  ;[1 1 1] Animate Blinds Material - Taken from parent. Reworked to stop the double multicast
  (if (<= (Utilities|Array|Length (Variables|Default|GetHidingPlayers)) 1)
    (Components|Timeline|Stop (Variables|BP_Evergreen_PhoneBooth|GetEnterBlindsAnim))
    (Rendering|Material|SetCustomPrimitiveDataFloat _staticmeshcomponent 6 1.0)
    (bind (_direction _float) (Timeline|EnterBlindsAnim PlayFromStart)
      (:Update
        (Rendering|Material|SetCustomPrimitiveDataFloat _staticmeshcomponent 5 _float))
      (:Finished
        (Variables|Default|SetBoothOcupied 1)))))

(event Custom|PhoneBoothBlindsExit
  ;[1 1 1] Animate Blinds Material - Taken from parent. Reworked to stop the double multicast
  (Components|Timeline|Stop (Variables|BP_Evergreen_PhoneBooth|GetExitBlindsAnim))
  (bind (_direction _float) (Timeline|ExitBlindsAnim PlayFromStart)
    (:Update
      (Rendering|Material|SetCustomPrimitiveDataFloat (Variables|Debug|GetStaticMeshComponent) 5 _float))
    (:Finished
      (Variables|Default|SetBoothOcupied 0))))

(event Custom|PhoneBoothDoorSwingAnimation
  ;[1 1 1] Door Swing Animation -  Taken from parent. Reworked to stop the double multicast
  ;;[1 1 1 1 12] Fix for FORT-297163. Prevent server-replicated event from triggering 2nd door swing. Could still break if replication is delayed longer than it takes the door to close.
  (if (not (Components|Timeline|IsPlaying (Variables|BP_Evergreen_PhoneBooth|GetPhoneBoothDoorSwing)))
    (Collision|SetCollisionResponsetoChannel (Variables|Default|GetSM_Telephone_Booth_DoorGlass) "ECC_Camera" "ECR_Ignore")
    (bind (_direction _float) (Timeline|PhoneBoothDoorSwing PlayFromStart)
      (:Update
        (Transformation|SetRelativeRotation (Variables|Default|GetDoor) (Math|Rotator|Lerp(Rotator) "0,90.000000, 0" "0,-30.000000, 0" _float) false false))
      (:Finished
        (Components|Activation|Activate (Variables|Default|GetNS_PhantomBooth_DoorSlam) true)))))

(event Custom|EnterBoothAnimations
  ;[1 1 1] Made these two multicasts to so that we are only calling one multicast not 2 when entering and exiting
  (CallFunction|PhoneBoothDoorSwingAnimation)
  (CallFunction|PhoneBoothBlindsEnter))

(event Custom|ExitBoothAnimations
  ;[1 1 1] Made these two multicasts to so that we are only calling one multicast not 2 when entering and exiting
  (CallFunction|PhoneBoothDoorSwingAnimation)
  (CallFunction|PhoneBoothBlindsExit))

(event Custom|EnablePhonebooth
  (params (bEnable :type Boolean))
  (bind _output_get (Variables|Default|SetbIsPhoneboothEnabled bEnable))
  (if (not _output_get)
    (Utilities|FlowControl|Delay 1e+01)
    (Variables|Default|SetbIsPhoneboothEnabled true)))

(fn AddPawnAndLoadoutToArrays
  (params (FortPawn :type Fort%20Pawn%20Object%20Reference))
  (bind _asfort_player_pawn (Utilities|Casting|CastToFortPlayerPawn FortPawn))
  (Utilities|IsValid (CosmeticLoadout|GetItemDefinitionTypefromLoadoutwithActor _asfort_player_pawn "(TagName=\"LoadoutSlot.Character\")" "/Script/FortniteGame.AthenaCharacterItemDefinition")
    (:Is%20Valid
      (bind _disguisecid (Variables|Default|GetDisguiseCID)
        (:then
          (if (!= _disguisecid (CosmeticLoadout|GetItemDefinitionTypefromLoadoutwithActor _asfort_player_pawn "(TagName=\"LoadoutSlot.Character\")" "/Script/FortniteGame.AthenaCharacterItemDefinition"))
            (Utilities|Array|AddUnique (Variables|Default|GetInteractedPawnsLoadouts) (Utilities|Struct|MakeStructEvergreenPhoneboothCosmetics _asfort_player_pawn (CosmeticLoadout|GetItemDefinitionTypefromLoadoutwithActor _asfort_player_pawn "(TagName=\"LoadoutSlot.Character\")" "/Script/FortniteGame.AthenaCharacterItemDefinition")))))
        (:else)))))

(fn BlueprintCanInteract
  (params (InteractingPawn :type Fort%20Pawn%20Object%20Reference) (InteractionBeingAttempted :type EInteractionBeingAttempted%20Enum :default "FirstInteraction") (InteractionType :type TInteractionType%20Enum :default "IT_NoInteraction"))
  (Utilities|IsValid InteractingPawn
    (:Is%20Valid
      (bind _caninteractbox (Variables|Default|GetCanInteractBox)
        (:then
          (return (Math|Boolean|ANDBoolean (> (Utilities|Array|FindItem (Collision|GetOverlappingActors _caninteractbox 0 :!class /Script/Engine.PrimitiveComponent) InteractingPawn) -1) (not (GameplayTags|HasAnyMatchingGameplayTags InteractingPawn (Variables|Default|GetBlockedInteractTags))) (Variables|Default|GetbIsPhoneboothEnabled))))
        (:else
          _)))
    (:Is%20Not%20Valid
      _))
  (return false))

(fn GetValidOriginalCID
  (params (FortPawn :type Fort%20Pawn%20Object%20Reference))
  (bind _asfort_player_pawn (Utilities|Casting|CastToFortPlayerPawn FortPawn))
  (LocalVariables|SetL_FortPlayerPawn _asfort_player_pawn)
  (bind _array_element (Utilities|Array|ForEachLoopwithBreak (Variables|Default|GetInteractedPawnsLoadouts))
    (:LoopBody
      (bind _l_fortplayerpawn (LocalVariables|GetL_FortPlayerPawn)
        (:then
          (if (Math|Boolean|ANDBoolean (== _l_fortplayerpawn (Utilities|Struct|BreakStructEvergreenPhoneboothCosmetics _array_element)) (Utilities|IsValid (Utilities|Struct|BreakStructEvergreenPhoneboothCosmetics _array_element) :!class /Script/Engine.KismetSystemLibrary) (Utilities|IsValid (.CID_5_078EF0DF45EE8A5E2497B59D8A855FBD (Utilities|Struct|BreakStructEvergreenPhoneboothCosmetics _array_element)) :!class /Script/Engine.KismetSystemLibrary))
            (LocalVariables|SetL_Character (.CID_5_078EF0DF45EE8A5E2497B59D8A855FBD (Utilities|Struct|BreakStructEvergreenPhoneboothCosmetics _array_element)))
            (LocalVariables|SetL_Valid true)))
        (:else)))
    (:Completed
      _))
  (bind _l_character (LocalVariables|GetL_Character)
    (:then
      (return (LocalVariables|GetL_Valid) _l_character))
    (:else
      (return false 0))))

(fn IsPawnAlreadyDisguised
  (params (FortPawn :type Fort%20Pawn%20Object%20Reference))
  (bind _asfort_player_pawn (Utilities|Casting|CastToFortPlayerPawn FortPawn)
    (:then
      (Utilities|IsValid (CosmeticLoadout|GetItemDefinitionTypefromLoadoutwithActor _asfort_player_pawn "(TagName=\"LoadoutSlot.Character\")" "/Script/FortniteGame.AthenaCharacterItemDefinition")
        (:Is%20Valid
          (bind _disguisecid (Variables|Default|GetDisguiseCID)
            (:then
              (return (== _disguisecid (CosmeticLoadout|GetItemDefinitionTypefromLoadoutwithActor _asfort_player_pawn "(TagName=\"LoadoutSlot.Character\")" "/Script/FortniteGame.AthenaCharacterItemDefinition"))))
            (:else
              _)))
        (:Is%20Not%20Valid
          _))
      (return false))
    (:CastFailed
      _)))

(fn ConstructionScript
  (|CallParentFunction))
